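# Modal telling the player about an account ban, behavior warning or matchmaking penalty for the selected island.
# On activation it reads ban info from FortMcpContext, treating zero or negative duration as a warning.
# BanEnumToText builds localized reason text, with special wording for competitive and teaming-with-cheater cases.
# SetBansViewed disables Continue while awaiting acknowledgement; closing retries unacknowledged requests, removing the modal on success or after a capped try count.
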WidgetBlueprint /Game/UI/Foundation/Modals/PlayerReporting/PlayerBannedModal
graphs [
 BanEnumToText {nodes [
   n1 = BanEnumToText!() -> (
    then
    Index:EPlayerBanReasons Enum=Teaming
    ExploitProgramName:String
    Competitive Ban Reason:EPlayerCompetitiveBanReasons Enum=None
    bIsCompetitiveBan:bool=~
   ) {is=title K2Node_FunctionEntry}
   n2 = AND Boolean(
    A="integer > integer"(A=Len(S=@n1.ExploitProgramName), B=0) {is=title K2Node_CallFunction}
    B="Equal (Enum)"(A:EPlayerBanReasons Enum=@n1, B:EPlayerBanReasons Enum=Exploiting)
   )
   n3 = Format Text!(
    Format="LOCTEXT(\"{A}: {B}\")"
    A:Text=MakeLiteralText(Value="LOCTEXT(\"Running\")")
    B:Text="To Text (String)"(InString=@n1.ExploitProgramName)
   ) -> (Result)
   n4 = Select!(
    Teaming:Text="LOCTEXT(\"Teaming\")"
    Afk_Leeching:Text="LOCTEXT(\"AFK Leeching\")"
    Harassment:Text="LOCTEXT(\"Harassment\")"
    TradeScamming:Text="LOCTEXT(\"Trade Scamming\")"
    Exploiting:Text="LOCTEXT(\"Exploiting\")"
    Competitive:Text="LOCTEXT(\"Competitive\")"
    Creative:Text="LOCTEXT(\"Creative\")"
    TeamingWithCheater:Text
    StreamSniping:Text="LOCTEXT(\"Stream Sniping\")"
    ModifiedCosmetics:Text="LOCTEXT(\"Using third-party tools to unlock cosmetic items\")"
    TermsOfService:Text="LOCTEXT(\"Terms of Service Violation\")"
    Index:EPlayerBanReasons Enum=@n1
   ) -> (ReturnValue:Text)
   n5 = Select!(
    None:Text="LOCTEXT(\"None\")"
    Cheating:Text="LOCTEXT(\"Cheating\")"
    Collusion:Text="LOCTEXT(\"Teaming up with a competitor. Please remember to play fair in Fortnite!\")"
    Toxicity:Text="LOCTEXT(\"Toxicity\")"
    Harassment="LOCTEXT(\"Harassment\")"
    Ringing:Text="LOCTEXT(\"Ringing\")"
    Gambling:Text="LOCTEXT(\"Gambling\")"
    Exploiting="LOCTEXT(\"Exploiting\")"
    IntentionalDisconnect:Text="LOCTEXT(\"Leaving a match intentionally in order to grief another player. Please remember to play fair in Fortnite!\")"
    IllegalRestart:Text="LOCTEXT(\"Playing in a tournament with multiple accounts. Remember, only one entry per participant!\")"
    Other:Text="LOCTEXT(\"Other\")"
    AccountSharing:Text="LOCTEXT(\"Sharing your account with another participant while competing in a tournament.\")"
    CircumventingRegionLock:Text="LOCTEXT(\"Playing in multiple regions of an event that is limited to one region per participant.\")"
    CircumventingBan:Text="LOCTEXT(\"Playing on an alternate account after receiving a competitive matchmaking ban.\")"
    Smurfing:Text="LOCTEXT(\"Competing on alternate accounts/Smurfing\")"
    CircumventingTeamLock:Text="LOCTEXT(\"Playing with new teammates in a tournament that is limited to one team per participant.\")"
    RuleViolation:Text="LOCTEXT(\"Fortnite online tournament rule(s) violation.\")"
    TeamingWithCheater="LOCTEXT(\"Teaming up with a cheater\")"
    Index:EPlayerCompetitiveBanReasons Enum=@n1.Competitive Ban Reason
   ) -> (ReturnValue)
   n6 = Format Text!(
    Format="LOCTEXT(\"{CompetitiveType}: {CompetitiveReason}\")"
    CompetitiveType:Text=MakeLiteralText(Value="LOCTEXT(\"Competitive\")")
    CompetitiveReason:Text=@n5
   ) -> (Result)
   > Return Node!(
    execute
    ReturnValue:Text=Select!(
     Option 0:Text=Select(Option 0=@n4, Option 1=@n3, Index=@n2)
     Option 1:Text=Select(Option 0=@n6, Option 1=@n5, Index=@n1.bIsCompetitiveBan)
     Index:bool="Equal (Enum)"(A=@n1, B=Competitive)
    ) -> (ReturnValue)
   )
  ]}
 DisableButton {nodes [
   n1 = DisableButton!() -> (then, DisableButton:bool=~) {is=title K2Node_FunctionEntry}
   > SetIsEnabled(self=Get CloseButton() -> (CloseButton:Icon Text Button Object Reference), bInIsEnabled=NOT Boolean(A=@n1))
  ]}
 EventGraph {nodes [
   "On Button Clicked (CloseButton)"!() -> (OutputDelegate, then, Button:Common Button Legacy Object Reference) {is=title K2Node_ComponentBoundEvent}
   > HandleCloseAttempted()
   Event On Activated()
   > n1 = Branch(Condition=IsSelectedIslandBanned!(self:Fort Mcp Context Object Reference=Get FortMcpContext!() -> (ReturnValue:Fort Mcp Context Object Reference) {is=title K2Node_GetBlueprintContextEx}) -> (ReturnValue:bool))
   > SetupSelectedIslandBanned()
   Set McpBanInfo(
    execute=@n1.else
    McpBanInfo:Mcp Ban Info Structure=GetBanInfo!(self:Fort Mcp Context Object Reference=Get FortMcpContext()) -> (ReturnValue:"Mcp Ban Info Structure (by ref)")
   ) -> (Output_Get:Mcp Ban Info Structure)
   n2 = GetBanDuration!(
    self:Fort Mcp Context Object Reference=Get FortMcpContext()
    BanInfo:"Mcp Ban Info Structure (by ref)"=Get McpBanInfo() -> (McpBanInfo:Mcp Ban Info Structure)
   ) -> (ReturnValue:Timespan Structure)
   > Set bIsWarning(bIsWarning:bool="float <= float"!(A:double=Get Total Seconds!(A:Timespan Structure=@n2) -> (ReturnValue:double) {is=title K2Node_CallFunction}, B:double=0) -> (ReturnValue:bool) {is=title K2Node_CallFunction}) -> (Output_Get:bool)
   n3 = Break Mcp Ban Info!(McpBanInfo:"Mcp Ban Info Structure (by ref)"=Get McpBanInfo()) -> (
    BanReasons:Array of EPlayerBanReasons Enums
    ExploitProgramName:String
    AdditionalInfo:String
    CompetitiveBanReason:EPlayerCompetitiveBanReasons Enum
    bRequiresUserAck:bool=~
   ) {is=title K2Node_BreakStruct}
   n4 = Length(TargetArray:Array of EPlayerBanReasons Enums=@n3)
   > Set bIsCompetitiveBan(bIsCompetitiveBan:bool=AND Boolean(
     A="Equal (Integer)"(A=@n4, B=1) {is=title K2Node_CallFunction}
     B=Contains Item(TargetArray:Array of EPlayerBanReasons Enums=@n3, ItemToFind:"EPlayerBanReasons Enum (by ref)"=Competitive)
    )) -> (Output_Get:bool)
   > Set bIsTeamingWithCheaterBan(bIsTeamingWithCheaterBan:bool=Contains Item(TargetArray=@n3, ItemToFind=TeamingWithCheater)) -> (Output_Get:bool)
   > "SetText (Text)"(
    self=Get Title() -> (Title:Common Text Object Reference)
    InText=Select(Option 0="LOCTEXT(\"This account has been banned.\")", Option 1="LOCTEXT(\"Behavior Warning\")", Index=Get bIsWarning() -> (bIsWarning:bool))
   )
   > "SetText (Text)"(
    self=Get disclaimer() -> (disclaimer:Common Text Object Reference)
    InText=Select(
     Option 0="LOCTEXT(\"This temporary ban may precede a permanent ban depending on severity of the prohibited activity.\")"
     Option 1="LOCTEXT(\"Please refrain from repeat offenses or you may receive an account ban.\")"
     Index=Get bIsWarning()
    )
   )
   n5 = Get bIsWarning()
   n6 = Select(Option 0="LOCTEXT(\"banned\")", Option 1="LOCTEXT(\"warned\")", Index=@n5)
   n7 = Select(Option 0="LOCTEXT(\"banned from competitive play\")", Option 1="LOCTEXT(\"issued a competitive warning\")", Index=@n5)
   n8 = Get bIsCompetitiveBan() -> (bIsCompetitiveBan:bool)
   > Set BanInformation(BanInformation:Text=Format Text!(
     Format="LOCTEXT(\"Due to recent actions on your account, you have been {severity} for:\")"
     severity:Text=Select(Option 0=@n6, Option 1=@n7, Index=@n8)
    ) -> (Result)) -> (Output_Get:Text)
   > n9 = For Each Loop(Array:Array of EPlayerBanReasons Enums=@n3) -> (Array Element:EPlayerBanReasons Enum)
   > n10 = BanEnumToText!(
    execute
    self
    Index:EPlayerBanReasons Enum=@n9
    ExploitProgramName:String=@n3.ExploitProgramName
    Competitive Ban Reason:EPlayerCompetitiveBanReasons Enum=@n3.CompetitiveBanReason
    bIsCompetitiveBan:bool=Get bIsCompetitiveBan()
   ) -> (then, ReturnValue:Text) {is=title K2Node_CallFunction}
   > Set BanInformation(BanInformation=Format Text!(Format="LOCTEXT(\"{A}\\r\\n{B}\")", A:Text=Get BanInformation() -> (BanInformation:Text), B:Text=@n10) -> (Result))
   n11 = Branch(execute=@n9.Completed, Condition="integer > integer"(A=Len(S=@n3.AdditionalInfo), B=0))
   > n12 = Set BanInformation(BanInformation=Format Text(Format="LOCTEXT(\"{A}\\r\\n({B})\")", A=Get BanInformation(), B="To Text (String)"(InString=@n3.AdditionalInfo)))
   n13 = Branch(execute=[@n11.else, @n12.then], Condition=Get bIsTeamingWithCheaterBan() -> (bIsTeamingWithCheaterBan:bool))
   n14 = MakeLiteralText(Value="LOCTEXT(\"banned\")")
   n15 = MakeLiteralText(Value="LOCTEXT(\"warned\")")
   n16 = Get bIsWarning()
   > "SetText (Text)"(
    self=Get Title()
    InText=Format Text!(Format=MakeLiteralText(Value="LOCTEXT(\"You have been {0}!\")"), 0:Text=Select(Option 0=@n14, Option 1=@n15, Index=@n16)) -> (Result)
   )
   > n17 = Branch(Condition=@n16)
   n18 = Get BanMessage() -> (BanMessage:Text Object Reference)
   > n19 = "SetText (Text)"(
    self=@n18
    InText=MakeLiteralText(Value="LOCTEXT(\"You are receiving a warning as we\\'ve detected you repeatedly teaming up with a cheater, please remember to play fair in Fortnite!\")")
   )
   n20 = "SetText (Text)"(execute=[@n19.then, @n24.then], self=Get disclaimer())
   n21 = Get FortMcpContext()
   n22 = Get McpBanInfo()
   UpdateTimeRemaining!(
    execute=[@n25.then, @n20.then]
    self
    Time:Text=GetTimespanAsText!(
     Timespan:"Timespan Structure (by ref)"=GetBanDuration(self=@n21, BanInfo=@n22)
     bShowHoursWhenZero:bool=true
     bShowSeconds:bool=false
     FormatType:ETimespanAsTextFormat Enum=DaysHoursMinutesSeconds
     UnitDisplayType:ETimespanUnitDisplayFormat Enum=Full
     bUseSentenceCase:bool=false
    ) -> (ReturnValue:Text)
   ) -> (then)
   > n23 = Branch(Condition=Break Mcp Ban Info(McpBanInfo=Get McpBanInfo()) -> (bRequiresUserAck=~).bRequiresUserAck)
   > SetBansViewed!(execute, self:Fort Mcp Context Object Reference=Get FortMcpContext(), OnBanAcknowledged:Delegate=Create Event()) -> (then)
   > DisableButton!(execute, self, DisableButton:bool=true) -> (then) {is=title K2Node_CallFunction}
   Set AckSucceeded(execute=@n23.else, AckSucceeded:bool=true) -> (Output_Get:bool)
   n24 = "SetText (Text)"(
    execute=@n17.else
    self=@n18
    InText=MakeLiteralText(Value="LOCTEXT(\"You\\'ve been temporarily banned for repeatedly teaming up with a cheater, please remember to play fair in Fortnite!\")")
   )
   n25 = "SetText (Text)"(execute=@n13.else, self=Get BanMessage(), InText=Get BanInformation())
   Event PreConstruct()
   Event Construct()
   Event Tick()
   Make Array!("[0]":EPlayerBanReasons Enum=Teaming, "[1]":EPlayerBanReasons Enum=Harassment) -> (Array:Array of EPlayerBanReasons Enums)
  ]}
 HandleCloseAttempted {nodes [
   HandleCloseAttempted() {is=title K2Node_FunctionEntry}
   n1 = Get TryCount() -> (TryCount:int=0)
   > n2 = Branch(Condition=OR Boolean(
     A=Get AckSucceeded() -> (AckSucceeded:bool)
     B="integer >= integer"!(A:int=@n1, B:int=4) -> (ReturnValue:bool) {is=title K2Node_CallFunction}
    ))
   > RemoveModal()
   > Call On Close Succeeded!(execute:Exec, self:Player Banned Modal Object Reference) -> (then:Exec) {is=title K2Node_CallDelegate}
   SetBansViewed(execute=@n2.else, self=Get FortMcpContext(), OnBanAcknowledged=Create Event())
   > DisableButton(DisableButton=true)
   > Increment Int(Value=Get TryCount() -> (TryCount=0))
  ]}
 OnBanAck {nodes [
   n1 = OnBanAck!() -> (then, Success:bool=~) {is=title K2Node_FunctionEntry}
   > Set AckSucceeded(AckSucceeded=@n1)
   > DisableButton(DisableButton=false)
  ]}
 OnCloseSucceeded {nodes [OnCloseSucceeded() {is=title K2Node_FunctionEntry}]}
 RemoveModal {nodes [
   RemoveModal() {is=title K2Node_FunctionEntry}
   > PopActivatablePanelInModalLayer!(execute, Panel:Common Activatable Panel Legacy Object Reference=Self-Reference()) -> (then)
  ]}
 SetupSelectedIslandBanned {nodes [
   SetupSelectedIslandBanned() {is=title K2Node_FunctionEntry}
   n1 = Get FortMcpContext()
   > Set BannedIslandName(BannedIslandName:Text="To Text (String)"(InString=GetBannedIslandName!(self:Fort Mcp Context Object Reference=@n1) -> (ReturnValue:String))) -> (Output_Get:Text)
   > "SetText (Text)"(
    self=Get Title()
    InText=Format Text(Format="LOCTEXT(\"This account has Matchmaking Restrictions for {0}.\")", 0=Get BannedIslandName() -> (BannedIslandName:Text))
   )
   > "SetText (Text)"(
    self=Get BanMessage()
    InText=Format Text(
     Format="LOCTEXT(\"Due to your recent actions, you have received a Matchmaking Penalty for {0}: Leaving Match Early\")"
     0=Get BannedIslandName()
    )
   )
   > Set SelectedIslandBanDuration(SelectedIslandBanDuration:Timespan Structure=GetSelectedIslandBanDuration!(self:Fort Mcp Context Object Reference=Get FortMcpContext()) -> (ReturnValue:Timespan Structure)) -> (Output_Get:Timespan Structure)
   n2 = Get SelectedIslandBanDuration() -> (SelectedIslandBanDuration:Timespan Structure)
   > "SetText (Text)"(
    self=Get Timeremaining() -> (Timeremaining:Text Object Reference)
    InText=Format Text!(
     Format="LOCTEXT(\"Matchmaking Penalty Length: \\r\\n{x}\")"
     x:Text=GetTimespanAsText(Timespan=@n2, bShowHoursWhenZero=false, bShowSeconds=false, FormatType=DaysHoursMinutesSeconds, UnitDisplayType=Full, bUseSentenceCase=false)
    ) -> (Result)
   )
   n3 = Get Total Seconds(A=Get SelectedIslandBanDuration())
   > Set Visibility(
    Visibility=Select(Option 0:ESlateVisibility Enum=Collapsed, Option 1:ESlateVisibility Enum=HitTestInvisible, Index="float > float"(A=@n3)) -> (ReturnValue:ESlateVisibility Enum)
    self=Get Timeremaining()
   )
   > "SetText (Text)"(self=Get disclaimer(), InText="LOCTEXT(\"Your matchmaking Penalty length will increase, if you repeat offenses. \")")
   > SetBansViewed(self=Get FortMcpContext(), OnBanAcknowledged=Create Event())
   > DisableButton(DisableButton=true)
  ]}
 UpdateText {nodes [
   n1 = UpdateText!() -> (then, Title:Text, Content:Text) {is=title K2Node_FunctionEntry}
   > "SetText (Text)"(self=Get Title(), InText=@n1)
   > "SetText (Text)"(self=Get BanMessage(), InText=@n1.Content)
  ]}
 UpdateTimeRemaining {nodes [
   n1 = UpdateTimeRemaining!() -> (then, Time:Text) {is=title K2Node_FunctionEntry}
   n2 = Get Timeremaining()
   > "SetText (Text)"(self=@n2, InText=Format Text(Format="LOCTEXT(\"Ban Length: \\r\\n{x}\")", x=@n1))
   > SetVisibility(self=@n2, InVisibility=Select(Option 0=SelfHitTestInvisible, Option 1=Collapsed, Index=TextIsEmpty(InText=@n1)))
  ]}
]
variables [
 AckSucceeded {type=boolean}
 BanInformation {type=string, default=Due to your recent actions you have been banned for:}
 BanType {type=string, default=""}
 bIsCompetitiveBan {type=boolean}
 bIsTeamingWithCheaterBan {type=boolean}
 bIsWarning {type=boolean}
 McpBanInfo {
  type=McpBanInfo
  default {
   additionalInfo=""
   bBanHasStarted=false
   bRequiresUserAck=false
   banDurationDays=0
   banStartTimeUtc=0001.01.01-00.00.00
   competitiveBanReason=None
   exploitProgramName=""
   banReasons []
  }
 }
 OnCloseSucceeded {}
 TryCount {type=integer}
]
widgets {
 info {
  inheritedWidgetCount=0
  namedSlotCount=1
  parentClass=/Script/FortniteUI.FortPlayerBanned
  rootWidgetClass=&Overlay
  widgetCount=38
 }
 widgets [
  Overlay_0 {widgetClassPath=&Overlay, widget=Overlay_0}
  Lightbox {
   widgetClassPath=/Game/UI/Foundation/Lightbox/Lightbox
   bIsVariable=true
   parent=Overlay_0
   widget=Lightbox
   properties {dimAmount=0.8}
   slot {class=OverlaySlot, properties {horizontalAlignment=HAlign_Fill, verticalAlignment=VAlign_Fill}}
  }
  "[Border[ Dialog Content" {
   widgetClassPath=&Border
   namedSlotHost=Lightbox
   widget="[Border[ Dialog Content"
   properties {horizontalAlignment=HAlign_Center, verticalAlignment=VAlign_Center, brushColor=<rgba 0 0 0 0.25>, padding=<ltrb 0 0 0 0>}
   slot {class=OverlaySlot}
  }
  VerticalBox_1 {
   widgetClassPath=&VerticalBox
   parent="[Border[ Dialog Content"
   widget=VerticalBox_1
   slot {class=BorderSlot, properties {horizontalAlignment=HAlign_Center, verticalAlignment=VAlign_Center, padding=<ltrb 0 0 0 0>}}
  }
  SizeBox_1 {
   widgetClassPath=&SizeBox
   parent=VerticalBox_1
   widget=SizeBox_1
   properties {bOverride_WidthOverride=true, minDesiredHeight=600, widthOverride=1000}
   slot {class=VerticalBoxSlot}
  }
  Overlay_2 {widgetClassPath=&Overlay, parent=SizeBox_1, widget=Overlay_2, slot {class=SizeBoxSlot}}
  Border_DarkGradient {
   widgetClassPath=&Border
   parent=Overlay_2
   widget=Border_DarkGradient
   properties {
    background {
     bIsDynamicallyLoaded=false
     drawAs=Image
     imageType=NoImage
     mirroring=NoMirror
     resourceName=None
     resourceObject=/Game/UI/Foundation/Modals/ErrorWindow/Materials/M_UI_HorizontalFade_Border
     tiling=NoTile
     imageSize=<32 32>
     margin=<ltrb 0 0 0 0>
     outlineSettings {
      bUseBrushTransparency=false
      roundingType=HalfHeightRadius
      width=0
      color {colorUseRule=UseColor_Specified, specifiedColor=<rgba 0 0 0 0>}
      cornerRadii=<0 0 0 0>
     }
     tintColor {colorUseRule=UseColor_Specified, specifiedColor=<rgba 1 1 1 1>}
     uVRegion {bIsValid=false, max=<0 0>, min=<0 0>}
    }
    brushColor=<rgba 0 0 0 0.28>
   }
   slot {class=OverlaySlot, properties {horizontalAlignment=HAlign_Fill, verticalAlignment=VAlign_Fill, padding=<ltrb 0 65 0 65>}}
  }
  Triangles_1 {
   widgetClassPath=&Image
   bIsVariable=true
   parent=Overlay_2
   widget=Triangles_1
   properties {
    renderOpacity=0.4619
    visibility=SelfHitTestInvisible
    brush {
     bIsDynamicallyLoaded=false
     drawAs=Image
     imageType=NoImage
     mirroring=NoMirror
     resourceName=None
     resourceObject=/Game/UI/Foundation/Onboarding/Materials/M_FX_TriangleScroll-PartySuggestion
     tiling=NoTile
     imageSize=<104.0479 105.5286>
     margin=<ltrb 0 0 0 0>
     outlineSettings {
      bUseBrushTransparency=false
      roundingType=HalfHeightRadius
      width=0
      color {colorUseRule=UseColor_Specified, specifiedColor=<rgba 0 0 0 0>}
      cornerRadii=<0 0 0 0>
     }
     tintColor {colorUseRule=UseColor_Specified, specifiedColor=<rgba 1 1 1 1>}
     uVRegion {bIsValid=false, max=<0 0>, min=<0 0>}
    }
    colorAndOpacity=<rgba 1 1 1 0.8>
    renderTransform {angle=-90, scale=<1.1 1>, shear=<0 0>, translation=<0 -19>}
    renderTransformPivot=<0.25 0.5>
   }
   slot {class=OverlaySlot, properties {horizontalAlignment=HAlign_Center}}
  }
  Triangles_2 {
   widgetClassPath=&Image
   bIsVariable=true
   parent=Overlay_2
   widget=Triangles_2
   properties {
    renderOpacity=0.1
    visibility=SelfHitTestInvisible
    brush {
     bIsDynamicallyLoaded=false
     drawAs=Image
     imageType=NoImage
     mirroring=NoMirror
     resourceName=None
     resourceObject=&M_FX_TriangleScroll-PartySuggestion
     tiling=NoTile
     imageSize=<104.0479 105.5286>
     margin=<ltrb 0 0 0 0>
     outlineSettings {
      bUseBrushTransparency=false
      roundingType=HalfHeightRadius
      width=0
      color {colorUseRule=UseColor_Specified, specifiedColor=<rgba 0 0 0 0>}
      cornerRadii=<0 0 0 0>
     }
     tintColor {colorUseRule=UseColor_Specified, specifiedColor=<rgba 1 1 1 1>}
     uVRegion {bIsValid=false, max=<0 0>, min=<0 0>}
    }
    colorAndOpacity=<rgba 1 1 1 0.8>
    renderTransform {angle=-90, scale=<0.7 2.85>, shear=<0 0>, translation=<17 -10>}
    renderTransformPivot=<0.25 0.5>
   }
   slot {class=OverlaySlot, properties {horizontalAlignment=HAlign_Center}}
  }
  Triangles_3 {
   widgetClassPath=&Image
   bIsVariable=true
   parent=Overlay_2
   widget=Triangles_3
   properties {
    renderOpacity=0.4357
    visibility=SelfHitTestInvisible
    brush {
     bIsDynamicallyLoaded=false
     drawAs=Image
     imageType=NoImage
     mirroring=NoMirror
     resourceName=None
     resourceObject=&M_FX_TriangleScroll-PartySuggestion
     tiling=NoTile
     imageSize=<104.0479 105.5286>
     margin=<ltrb 0 0 0 0>
     outlineSettings {
      bUseBrushTransparency=false
      roundingType=HalfHeightRadius
      width=0
      color {colorUseRule=UseColor_Specified, specifiedColor=<rgba 0 0 0 0>}
      cornerRadii=<0 0 0 0>
     }
     tintColor {colorUseRule=UseColor_Specified, specifiedColor=<rgba 1 1 1 1>}
     uVRegion {bIsValid=false, max=<0 0>, min=<0 0>}
    }
    colorAndOpacity=<rgba 1 1 1 0.8>
    renderTransform {angle=90, scale=<1 0.7>, shear=<0 0>, translation=<-17 37>}
   }
   slot {class=OverlaySlot, properties {horizontalAlignment=HAlign_Center, verticalAlignment=VAlign_Bottom}}
  }
  Border_1 {
   widgetClassPath=&Border
   parent=Overlay_2
   widget=Border_1
   properties {
    background {
     bIsDynamicallyLoaded=false
     drawAs=Image
     imageType=NoImage
     mirroring=NoMirror
     resourceName=None
     resourceObject=/Game/UI/Foundation/Modals/ErrorWindow/Materials/M_UI_ErrorPopUpBorder
     tiling=NoTile
     imageSize=<32 32>
     margin=<ltrb 0 0 0 0>
     outlineSettings {
      bUseBrushTransparency=false
      roundingType=HalfHeightRadius
      width=0
      color {colorUseRule=UseColor_Specified, specifiedColor=<rgba 0 0 0 0>}
      cornerRadii=<0 0 0 0>
     }
     tintColor {colorUseRule=UseColor_Specified, specifiedColor=<rgba 1 1 1 1>}
     uVRegion {bIsValid=false, max=<0 0>, min=<0 0>}
    }
    brushColor=<rgba 0.7656 0.8728 1 1>
    padding=<ltrb 4 80 4 80>
   }
   slot {class=OverlaySlot, properties {horizontalAlignment=HAlign_Fill, verticalAlignment=VAlign_Fill}}
  }
  VerticalBox_354 {
   widgetClassPath=&VerticalBox
   parent=Border_1
   widget=VerticalBox_354
   slot {class=BorderSlot, properties {padding=<ltrb 4 80 4 80>}}
  }
  WidgetSwitcher_SeverityImage {
   widgetClassPath=&WidgetSwitcher
   bIsVariable=true
   parent=VerticalBox_354
   widget=WidgetSwitcher_SeverityImage
   slot {class=VerticalBoxSlot, properties {horizontalAlignment=HAlign_Center, verticalAlignment=VAlign_Center}}
  }
  Image_Severe {
   widgetClassPath=&Image
   bIsVariable=true
   parent=WidgetSwitcher_SeverityImage
   widget=Image_Severe
   properties {
    brush {
     bIsDynamicallyLoaded=false
     drawAs=Image
     imageType=NoImage
     mirroring=NoMirror
     resourceName=None
     resourceObject=/Game/UI/Foundation/Shell/Textures/T-Icon-HardError-64
     tiling=NoTile
     imageSize=<64 64>
     margin=<ltrb 0 0 0 0>
     outlineSettings {
      bUseBrushTransparency=false
      roundingType=HalfHeightRadius
      width=0
      color {colorUseRule=UseColor_Specified, specifiedColor=<rgba 0 0 0 0>}
      cornerRadii=<0 0 0 0>
     }
     tintColor {colorUseRule=UseColor_Specified, specifiedColor=<rgba 1 1 1 1>}
     uVRegion {bIsValid=false, max=<0 0>, min=<0 0>}
    }
    colorAndOpacity=<rgba 1 0.0597 0.0199 1>
   }
   slot {class=WidgetSwitcherSlot}
  }
  VerticalBox_0 {
   widgetClassPath=&VerticalBox
   parent=VerticalBox_354
   widget=VerticalBox_0
   slot {class=VerticalBoxSlot, properties {verticalAlignment=VAlign_Center, padding=<ltrb 0 10 0 0>}}
  }
  Title {
   widgetClassPath=&CommonTextBlock
   bIsVariable=true
   parent=VerticalBox_0
   widget=Title
   properties {
    justification=Center
    style=/FortUILibrary/Styles/Typography/Foundation/TextStyle-Header-L.TextStyle-Header-L_C
    text="You have been banned[ph]"
    font {
     bForceMonospaced=false
     bMaterialIsStencil=false
     fontMaterial=null
     fontObject=&Burbank
     letterSpacing=0
     monospacedWidth=1
     size=48
     skewAmount=0
     typefaceFontName=Burbank-Black
     outlineSettings {
      bApplyOutlineToDropShadows=false
      bMiteredCorners=false
      bSeparateFillAlpha=false
      outlineMaterial=null
      outlineSize=0
      outlineColor=<rgba 0 0 0 1>
     }
    }
    shadowOffset=<0 0>
   }
   slot {class=VerticalBoxSlot}
  }
  BanMessage {
   widgetClassPath=&TextBlock
   bIsVariable=true
   parent=VerticalBox_0
   widget=BanMessage
   properties {
    justification=Center
    text="[ph]Due to recent actions on your account, you have been banned for:\r\nTrade Scam"
    wrapTextAt=900
    font {
     bForceMonospaced=false
     bMaterialIsStencil=false
     fontMaterial=null
     fontObject=&NotoSans
     letterSpacing=0
     monospacedWidth=1
     size=22
     skewAmount=0
     typefaceFontName=NotoSans-Regular
     outlineSettings {
      bApplyOutlineToDropShadows=false
      bMiteredCorners=false
      bSeparateFillAlpha=false
      outlineMaterial=null
      outlineSize=0
      outlineColor=<rgba 0 0 0 1>
     }
    }
   }
   slot {class=VerticalBoxSlot, properties {horizontalAlignment=HAlign_Center, padding=<ltrb 0 10 0 0>}}
  }
  Timeremaining {
   widgetClassPath=&TextBlock
   bIsVariable=true
   parent=VerticalBox_0
   widget=Timeremaining
   properties {
    justification=Center
    text="Ban Time Remaining:\r\nxd"
    font {
     bForceMonospaced=false
     bMaterialIsStencil=false
     fontMaterial=null
     fontObject=&NotoSans
     letterSpacing=0
     monospacedWidth=1
     size=24
     skewAmount=0
     typefaceFontName=NotoSans-Bold
     outlineSettings {
      bApplyOutlineToDropShadows=false
      bMiteredCorners=false
      bSeparateFillAlpha=false
      outlineMaterial=null
      outlineSize=0
      outlineColor=<rgba 0 0 0 1>
     }
    }
   }
   slot {class=VerticalBoxSlot, properties {horizontalAlignment=HAlign_Center, padding=<ltrb 0 25 0 0>}}
  }
  disclaimer {
   widgetClassPath=&CommonTextBlock
   bIsVariable=true
   parent=VerticalBox_0
   widget=disclaimer
   properties {
    justification=Center
    style=/FortUILibrary/Styles/Typography/Foundation/TextStyle-Base-M-I.TextStyle-Base-M-I_C
    text="[ph]This temporary ban may precede a permanent ban depending on severity of the prohibited activity."
    wrapTextAt=700
    font {
     bForceMonospaced=false
     bMaterialIsStencil=false
     fontMaterial=null
     fontObject=&NotoSans
     letterSpacing=0
     monospacedWidth=1
     size=24
     skewAmount=0
     typefaceFontName=NotoSans-Italic
     outlineSettings {
      bApplyOutlineToDropShadows=false
      bMiteredCorners=false
      bSeparateFillAlpha=false
      outlineMaterial=null
      outlineSize=0
      outlineColor=<rgba 0 0 0 1>
     }
    }
    shadowOffset=<0 0>
   }
   slot {class=VerticalBoxSlot, properties {horizontalAlignment=HAlign_Center, padding=<ltrb 0 10 0 0>}}
  }
  CloseButton {
   widgetClassPath=&IconTextButton
   bIsVariable=true
   parent=VerticalBox_0
   widget=CloseButton
   properties {
    bHideInputActionWithKeyboard=true
    buttonText=Continue
    onButtonBaseClicked=(&self:WidgetTree.CloseButton.HandleOnButtonClicked)
    onButtonBaseDoubleClicked=(&self:WidgetTree.CloseButton.HandleOnButtonDoubleClicked)
    onButtonBaseHovered=(&self:WidgetTree.CloseButton.HandleOnButtonHovered)
    onButtonBaseUnhovered=(&self:WidgetTree.CloseButton.HandleOnButtonUnhovered)
    onSelectedChangedBase=(&self:WidgetTree.CloseButton.HandleOnSelectedChanged)
    style=/FortUILibrary/Foundation/Buttons/ButtonStyles/Skew/ButtonStyle-Skew_Error.ButtonStyle-Skew_Error_C
    triggeringInputAction {dataTable=&InputActionDataTable, rowName=Continue}
   }
   slot {class=VerticalBoxSlot, properties {horizontalAlignment=HAlign_Center, verticalAlignment=VAlign_Bottom, padding=<ltrb 20 20 20 20>}}
  }
  "[Border[ Dialog Content" {
   widgetClassPath=&Border
   parent=Overlay_0
   widget="[Border[ Dialog Content"
   properties {horizontalAlignment=HAlign_Center, verticalAlignment=VAlign_Center, brushColor=<rgba 0 0 0 0.25>, padding=<ltrb 0 0 0 0>}
   slot {class=OverlaySlot}
  }
  VerticalBox_1 {
   widgetClassPath=&VerticalBox
   parent="[Border[ Dialog Content"
   widget=VerticalBox_1
   slot {class=BorderSlot, properties {horizontalAlignment=HAlign_Center, verticalAlignment=VAlign_Center, padding=<ltrb 0 0 0 0>}}
  }
  SizeBox_1 {
   widgetClassPath=&SizeBox
   parent=VerticalBox_1
   widget=SizeBox_1
   properties {bOverride_WidthOverride=true, minDesiredHeight=600, widthOverride=1000}
   slot {class=VerticalBoxSlot}
  }
  Overlay_2 {widgetClassPath=&Overlay, parent=SizeBox_1, widget=Overlay_2, slot {class=SizeBoxSlot}}
  Border_DarkGradient {
   widgetClassPath=&Border
   parent=Overlay_2
   widget=Border_DarkGradient
   properties {
    background {
     bIsDynamicallyLoaded=false
     drawAs=Image
     imageType=NoImage
     mirroring=NoMirror
     resourceName=None
     resourceObject=&M_UI_HorizontalFade_Border
     tiling=NoTile
     imageSize=<32 32>
     margin=<ltrb 0 0 0 0>
     outlineSettings {
      bUseBrushTransparency=false
      roundingType=HalfHeightRadius
      width=0
      color {colorUseRule=UseColor_Specified, specifiedColor=<rgba 0 0 0 0>}
      cornerRadii=<0 0 0 0>
     }
     tintColor {colorUseRule=UseColor_Specified, specifiedColor=<rgba 1 1 1 1>}
     uVRegion {bIsValid=false, max=<0 0>, min=<0 0>}
    }
    brushColor=<rgba 0 0 0 0.28>
   }
   slot {class=OverlaySlot, properties {horizontalAlignment=HAlign_Fill, verticalAlignment=VAlign_Fill, padding=<ltrb 0 65 0 65>}}
  }
  Triangles_1 {
   widgetClassPath=&Image
   bIsVariable=true
   parent=Overlay_2
   widget=Triangles_1
   properties {
    renderOpacity=0.4619
    visibility=SelfHitTestInvisible
    brush {
     bIsDynamicallyLoaded=false
     drawAs=Image
     imageType=NoImage
     mirroring=NoMirror
     resourceName=None
     resourceObject=&M_FX_TriangleScroll-PartySuggestion
     tiling=NoTile
     imageSize=<104.0479 105.5286>
     margin=<ltrb 0 0 0 0>
     outlineSettings {
      bUseBrushTransparency=false
      roundingType=HalfHeightRadius
      width=0
      color {colorUseRule=UseColor_Specified, specifiedColor=<rgba 0 0 0 0>}
      cornerRadii=<0 0 0 0>
     }
     tintColor {colorUseRule=UseColor_Specified, specifiedColor=<rgba 1 1 1 1>}
     uVRegion {bIsValid=false, max=<0 0>, min=<0 0>}
    }
    colorAndOpacity=<rgba 1 1 1 0.8>
    renderTransform {angle=-90, scale=<1.1 1>, shear=<0 0>, translation=<0 -19>}
    renderTransformPivot=<0.25 0.5>
   }
   slot {class=OverlaySlot, properties {horizontalAlignment=HAlign_Center}}
  }
  Triangles_2 {
   widgetClassPath=&Image
   bIsVariable=true
   parent=Overlay_2
   widget=Triangles_2
   properties {
    renderOpacity=0.1
    visibility=SelfHitTestInvisible
    brush {
     bIsDynamicallyLoaded=false
     drawAs=Image
     imageType=NoImage
     mirroring=NoMirror
     resourceName=None
     resourceObject=&M_FX_TriangleScroll-PartySuggestion
     tiling=NoTile
     imageSize=<104.0479 105.5286>
     margin=<ltrb 0 0 0 0>
     outlineSettings {
      bUseBrushTransparency=false
      roundingType=HalfHeightRadius
      width=0
      color {colorUseRule=UseColor_Specified, specifiedColor=<rgba 0 0 0 0>}
      cornerRadii=<0 0 0 0>
     }
     tintColor {colorUseRule=UseColor_Specified, specifiedColor=<rgba 1 1 1 1>}
     uVRegion {bIsValid=false, max=<0 0>, min=<0 0>}
    }
    colorAndOpacity=<rgba 1 1 1 0.8>
    renderTransform {angle=-90, scale=<0.7 2.85>, shear=<0 0>, translation=<17 -10>}
    renderTransformPivot=<0.25 0.5>
   }
   slot {class=OverlaySlot, properties {horizontalAlignment=HAlign_Center}}
  }
  Triangles_3 {
   widgetClassPath=&Image
   bIsVariable=true
   parent=Overlay_2
   widget=Triangles_3
   properties {
    renderOpacity=0.4357
    visibility=SelfHitTestInvisible
    brush {
     bIsDynamicallyLoaded=false
     drawAs=Image
     imageType=NoImage
     mirroring=NoMirror
     resourceName=None
     resourceObject=&M_FX_TriangleScroll-PartySuggestion
     tiling=NoTile
     imageSize=<104.0479 105.5286>
     margin=<ltrb 0 0 0 0>
     outlineSettings {
      bUseBrushTransparency=false
      roundingType=HalfHeightRadius
      width=0
      color {colorUseRule=UseColor_Specified, specifiedColor=<rgba 0 0 0 0>}
      cornerRadii=<0 0 0 0>
     }
     tintColor {colorUseRule=UseColor_Specified, specifiedColor=<rgba 1 1 1 1>}
     uVRegion {bIsValid=false, max=<0 0>, min=<0 0>}
    }
    colorAndOpacity=<rgba 1 1 1 0.8>
    renderTransform {angle=90, scale=<1 0.7>, shear=<0 0>, translation=<-17 37>}
   }
   slot {class=OverlaySlot, properties {horizontalAlignment=HAlign_Center, verticalAlignment=VAlign_Bottom}}
  }
  Border_1 {
   widgetClassPath=&Border
   parent=Overlay_2
   widget=Border_1
   properties {
    background {
     bIsDynamicallyLoaded=false
     drawAs=Image
     imageType=NoImage
     mirroring=NoMirror
     resourceName=None
     resourceObject=&M_UI_ErrorPopUpBorder
     tiling=NoTile
     imageSize=<32 32>
     margin=<ltrb 0 0 0 0>
     outlineSettings {
      bUseBrushTransparency=false
      roundingType=HalfHeightRadius
      width=0
      color {colorUseRule=UseColor_Specified, specifiedColor=<rgba 0 0 0 0>}
      cornerRadii=<0 0 0 0>
     }
     tintColor {colorUseRule=UseColor_Specified, specifiedColor=<rgba 1 1 1 1>}
     uVRegion {bIsValid=false, max=<0 0>, min=<0 0>}
    }
    brushColor=<rgba 0.7656 0.8728 1 1>
    padding=<ltrb 4 80 4 80>
   }
   slot {class=OverlaySlot, properties {horizontalAlignment=HAlign_Fill, verticalAlignment=VAlign_Fill}}
  }
  VerticalBox_354 {
   widgetClassPath=&VerticalBox
   parent=Border_1
   widget=VerticalBox_354
   slot {class=BorderSlot, properties {padding=<ltrb 4 80 4 80>}}
  }
  WidgetSwitcher_SeverityImage {
   widgetClassPath=&WidgetSwitcher
   bIsVariable=true
   parent=VerticalBox_354
   widget=WidgetSwitcher_SeverityImage
   slot {class=VerticalBoxSlot, properties {horizontalAlignment=HAlign_Center, verticalAlignment=VAlign_Center}}
  }
  Image_Severe {
   widgetClassPath=&Image
   bIsVariable=true
   parent=WidgetSwitcher_SeverityImage
   widget=Image_Severe
   properties {
    brush {
     bIsDynamicallyLoaded=false
     drawAs=Image
     imageType=NoImage
     mirroring=NoMirror
     resourceName=None
     resourceObject=&T-Icon-HardError-64
     tiling=NoTile
     imageSize=<64 64>
     margin=<ltrb 0 0 0 0>
     outlineSettings {
      bUseBrushTransparency=false
      roundingType=HalfHeightRadius
      width=0
      color {colorUseRule=UseColor_Specified, specifiedColor=<rgba 0 0 0 0>}
      cornerRadii=<0 0 0 0>
     }
     tintColor {colorUseRule=UseColor_Specified, specifiedColor=<rgba 1 1 1 1>}
     uVRegion {bIsValid=false, max=<0 0>, min=<0 0>}
    }
    colorAndOpacity=<rgba 1 0.0597 0.0199 1>
   }
   slot {class=WidgetSwitcherSlot}
  }
  VerticalBox_0 {
   widgetClassPath=&VerticalBox
   parent=VerticalBox_354
   widget=VerticalBox_0
   slot {class=VerticalBoxSlot, properties {verticalAlignment=VAlign_Center, padding=<ltrb 0 10 0 0>}}
  }
  Title {
   widgetClassPath=&CommonTextBlock
   bIsVariable=true
   parent=VerticalBox_0
   widget=Title
   properties {
    justification=Center
    style=&TextStyle-Header-L_C
    text="You have been banned[ph]"
    font {
     bForceMonospaced=false
     bMaterialIsStencil=false
     fontMaterial=null
     fontObject=&Burbank
     letterSpacing=0
     monospacedWidth=1
     size=48
     skewAmount=0
     typefaceFontName=Burbank-Black
     outlineSettings {
      bApplyOutlineToDropShadows=false
      bMiteredCorners=false
      bSeparateFillAlpha=false
      outlineMaterial=null
      outlineSize=0
      outlineColor=<rgba 0 0 0 1>
     }
    }
    shadowOffset=<0 0>
   }
   slot {class=VerticalBoxSlot}
  }
  BanMessage {
   widgetClassPath=&TextBlock
   bIsVariable=true
   parent=VerticalBox_0
   widget=BanMessage
   properties {
    justification=Center
    text="[ph]Due to recent actions on your account, you have been banned for:\r\nTrade Scam"
    wrapTextAt=900
    font {
     bForceMonospaced=false
     bMaterialIsStencil=false
     fontMaterial=null
     fontObject=&NotoSans
     letterSpacing=0
     monospacedWidth=1
     size=22
     skewAmount=0
     typefaceFontName=NotoSans-Regular
     outlineSettings {
      bApplyOutlineToDropShadows=false
      bMiteredCorners=false
      bSeparateFillAlpha=false
      outlineMaterial=null
      outlineSize=0
      outlineColor=<rgba 0 0 0 1>
     }
    }
   }
   slot {class=VerticalBoxSlot, properties {horizontalAlignment=HAlign_Center, padding=<ltrb 0 10 0 0>}}
  }
  Timeremaining {
   widgetClassPath=&TextBlock
   bIsVariable=true
   parent=VerticalBox_0
   widget=Timeremaining
   properties {
    justification=Center
    text="Ban Time Remaining:\r\nxd"
    font {
     bForceMonospaced=false
     bMaterialIsStencil=false
     fontMaterial=null
     fontObject=&NotoSans
     letterSpacing=0
     monospacedWidth=1
     size=24
     skewAmount=0
     typefaceFontName=NotoSans-Bold
     outlineSettings {
      bApplyOutlineToDropShadows=false
      bMiteredCorners=false
      bSeparateFillAlpha=false
      outlineMaterial=null
      outlineSize=0
      outlineColor=<rgba 0 0 0 1>
     }
    }
   }
   slot {class=VerticalBoxSlot, properties {horizontalAlignment=HAlign_Center, padding=<ltrb 0 25 0 0>}}
  }
  disclaimer {
   widgetClassPath=&CommonTextBlock
   bIsVariable=true
   parent=VerticalBox_0
   widget=disclaimer
   properties {
    justification=Center
    style=&TextStyle-Base-M-I_C
    text="[ph]This temporary ban may precede a permanent ban depending on severity of the prohibited activity."
    wrapTextAt=700
    font {
     bForceMonospaced=false
     bMaterialIsStencil=false
     fontMaterial=null
     fontObject=&NotoSans
     letterSpacing=0
     monospacedWidth=1
     size=24
     skewAmount=0
     typefaceFontName=NotoSans-Italic
     outlineSettings {
      bApplyOutlineToDropShadows=false
      bMiteredCorners=false
      bSeparateFillAlpha=false
      outlineMaterial=null
      outlineSize=0
      outlineColor=<rgba 0 0 0 1>
     }
    }
    shadowOffset=<0 0>
   }
   slot {class=VerticalBoxSlot, properties {horizontalAlignment=HAlign_Center, padding=<ltrb 0 10 0 0>}}
  }
  CloseButton {
   widgetClassPath=&IconTextButton
   bIsVariable=true
   parent=VerticalBox_0
   widget=CloseButton
   properties {
    bHideInputActionWithKeyboard=true
    buttonText=Continue
    onButtonBaseClicked=(&self:WidgetTree.CloseButton.HandleOnButtonClicked)
    onButtonBaseDoubleClicked=(&self:WidgetTree.CloseButton.HandleOnButtonDoubleClicked)
    onButtonBaseHovered=(&self:WidgetTree.CloseButton.HandleOnButtonHovered)
    onButtonBaseUnhovered=(&self:WidgetTree.CloseButton.HandleOnButtonUnhovered)
    onSelectedChangedBase=(&self:WidgetTree.CloseButton.HandleOnSelectedChanged)
    style=&ButtonStyle-Skew_Error_C
    triggeringInputAction {dataTable=&InputActionDataTable, rowName=Continue}
   }
   slot {class=VerticalBoxSlot, properties {horizontalAlignment=HAlign_Center, verticalAlignment=VAlign_Bottom, padding=<ltrb 20 20 20 20>}}
  }
 ]
}
